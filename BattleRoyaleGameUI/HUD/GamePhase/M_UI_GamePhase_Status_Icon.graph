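# UI-domain translucent material for the game phase status icon in the Battle Royale HUD.
# A static channel mask parameter selects one icon from the RGB channels of a shared icon texture.
# The mask is multiplied by an Opacity parameter, which a low-end value switcher function reduces slightly on its LowEnd input.
# A shared adaptive colors material function is also included for coloring.
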Material /BattleRoyaleGameUI/HUD/GamePhase/M_UI_GamePhase_Status_Icon
graph {
 properties {bAutomaticallySetUsageInEditor=false, blendMode=BLEND_Translucent, materialDomain=MD_UI, usedShadingModels=Default Lit}
 nodes [
  MaterialFunctionCall!(BufferSource) {materialFunction=/BattleRoyaleGameUI/Shared/MF_UI_AdaptiveColors}
  n1 = ScalarParameter() {defaultValue=1, parameterName=Opacity}
  Multiply(
   A=StaticComponentMaskParameter(None=TextureSample() {samplerType=SAMPLERTYPE_LinearColor, texture=&/T_UI_GamePhaseStatus_Icons_128x}.RGB) {defaultR=true, parameterName=Channel}
   B=MaterialFunctionCall!(Default=@n1, LowEnd=Multiply(A=@n1) {constB=0.9}) {materialFunction=/BattleRoyaleGameUI/Shared/MF_UI_LowEnd_ValueSwitcher}.Result
  )
 ]
}
