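# Translucent UI material that draws a perspective-skewed box backing, apparently for the side of the game-phase HUD.
# Perspective-warped UVs feed a box SDF that gives both a solid fill and a soft glow with parameterized threshold, smoothing and opacity.
# The glow fades by vertical screen-element UV and shows outside the fill, whose low-end variant adds a subtle vertical gradient.
# Color comes from the shared adaptive-colors white, with both gradient ends identical.
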Material /BattleRoyaleGameUI/HUD/GamePhase/M_UI_PhaseSideBacking
graph {
 properties {bAutomaticallySetUsageInEditor=false, blendMode=BLEND_Translucent, materialDomain=MD_UI, usedShadingModels=Default Lit}
 nodes [
  n1 = MaterialFunctionCall!(BufferSource) {materialFunction=/BattleRoyaleGameUI/Shared/MF_UI_AdaptiveColors}
  n2 = MaterialFunctionCall!(UV, Perspective Rotation=Constant() {r=0.4}, Perspective Amount=Constant() {r=0.1}) {materialFunction=/FortUILibrary/MaterialFunctions/Basics/MF_UI_Perspective}
  LinearInterpolate(A=@n1.White, B=@n1.White, Alpha=ComponentMask(None=@n2.Perspective UVs) {g=true})
  n3 = MaterialFunctionCall(UVs=Subtract(A=@n2.Perspective UVs, B=Constant2Vector() {g=0.025, r=0.025}), Size=Constant() {r=0.9}) {materialFunction=&MF_UI_Box}
  n4 = MaterialFunctionCall(
   DistanceField=@n3.SDF
   Threshold=ScalarParameter() {parameterName=Glow_Threshold}
   Smoothing=ScalarParameter() {defaultValue=0.25, parameterName=Glow_Smoothing}
  ) {materialFunction=&MF_UI_AntiAliasing}
  n5 = MaterialFunctionCall() {materialFunction=&GetUserInterfaceUV}
  n6 = NamedRerouteDeclaration(None=OneMinus(None=ComponentMask(None=@n5.Normalized UV) {g=true})) {name=VerticalMask, nodeColor=<rgba 0 0 0 1>}
  n7 = LinearInterpolate(Alpha=OneMinus(None=ComponentMask(None=@n2.Perspective UVs) {g=true})) {constA=0.85}
  LinearInterpolate(
   A=Multiply(A=Multiply(A=@n4.Inner, B=@n6), B=ScalarParameter() {defaultValue=0.5, parameterName=Glow_Opacity})
   B=MaterialFunctionCall!(Default=@n3.Fill, LowEnd=Multiply(A=@n3.Fill, B=@n7)) {materialFunction=/BattleRoyaleGameUI/Shared/MF_UI_LowEnd_ValueSwitcher}.Result
   Alpha=@n3.Fill
  )
 ]
}
