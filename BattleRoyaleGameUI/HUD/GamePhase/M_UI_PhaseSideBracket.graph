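# Translucent UI material that draws an angled, horizontally flippable side bracket for the game phase HUD.
# The Shape mask intersects top, bottom and outside BoxAngle cuts, then subtracts an inner angle limited by an antialiased, sine-based center cut.
# Anim_Stretch shifts the top and bottom angles, while Anim_Color blends adaptive white toward the dim color.
# Low-end value switchers include a branch that slightly scales down a rerouted value.
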Material /BattleRoyaleGameUI/HUD/GamePhase/M_UI_PhaseSideBracket
graph {
 properties {bAutomaticallySetUsageInEditor=false, blendMode=BLEND_Translucent, materialDomain=MD_UI, usedShadingModels=Default Lit}
 nodes [
  n1 = MaterialFunctionCall!(BufferSource) {materialFunction=/BattleRoyaleGameUI/Shared/MF_UI_AdaptiveColors}
  LinearInterpolate(
   A=@n1.White
   B=MaterialFunctionCall!(Default=@n1.Dim, LowEnd=@n1.Dim) {materialFunction=/BattleRoyaleGameUI/Shared/MF_UI_LowEnd_ValueSwitcher}.Result
   Alpha=ScalarParameter() {defaultValue=1, parameterName=Anim_Color}
  )
  n2 = NamedRerouteUsage()
  MaterialFunctionCall(Default=@n2, LowEnd=Multiply(A=@n2) {constB=0.9}) {materialFunction=&MF_UI_LowEnd_ValueSwitcher}
  n3 = MaterialFunctionCall() {materialFunction=&GetUserInterfaceUV}
  n4 = MaterialFunctionCall(Float2=@n3.Normalized UV) {materialFunction=&BreakOutFloat2Components}
  n5 = LinearInterpolate(A=@n3.Normalized UV, B=AppendVector(A=OneMinus(None=@n4.R), B=@n4.G), Alpha=ScalarParameter() {parameterName=FlipX})
  n6 = Constant()
  n7 = ScalarParameter() {defaultValue=11, parameterName=TopAngle}
  n8 = LinearInterpolate(
   A=@n7
   B=Subtract(A=@n7, B=ScalarParameter() {defaultValue=3, parameterName=StretchAmount_Top}) {constB=3}
   Alpha=ScalarParameter() {parameterName=Anim_Stretch}
  )
  n9 = MaterialFunctionCall(Float2=@n5) {materialFunction=&BreakOutFloat2Components}
  n10 = AppendVector(A=@n9.R, B=OneMinus(None=@n9.G))
  n11 = Constant()
  n12 = ScalarParameter() {defaultValue=13, parameterName=BottomAngle}
  n13 = LinearInterpolate(
   A=@n12
   B=Subtract(A=@n12, B=ScalarParameter() {defaultValue=5, parameterName=StretchAmount_Bottom}) {constB=3}
   Alpha=ScalarParameter() {parameterName=Anim_Stretch}
  )
  n14 = Min(
   A=Saturate(None=MaterialFunctionCall!(UV=@n5, Corner=@n6, Angle=@n8, Flip, Falloff) {materialFunction=/FortUILibrary/MaterialFunctions/Shapes/MF_UI_BoxAngle}.Result)
   B=Saturate(None=MaterialFunctionCall(UV=@n10, Corner=@n11, Angle=@n13) {materialFunction=&MF_UI_BoxAngle}.Result)
  )
  n15 = Saturate(None=MaterialFunctionCall(UV=@n10, Corner=Constant() {r=3}, Angle=ScalarParameter() {defaultValue=16, parameterName=OustideAngle}) {materialFunction=&MF_UI_BoxAngle}.Result)
  n16 = Add(A=@n10, B=ScalarParameter() {defaultValue=-0.15, parameterName=InnerAngleOffset})
  n17 = Constant() {r=3}
  n18 = ScalarParameter() {defaultValue=12, parameterName=InnerAngle}
  n19 = ComponentMask(None=Sine(None=@n5) {period=2}) {g=true}
  n20 = ScalarParameter() {defaultValue=0.475, parameterName=CenterCutHeight}
  n21 = Min(
   A=Saturate(None=MaterialFunctionCall(UV=@n16, Corner=@n17, Angle=@n18) {materialFunction=&MF_UI_BoxAngle}.Result)
   B=Saturate(None=MaterialFunctionCall(DistanceField=@n19, Threshold=@n20) {materialFunction=&MF_UI_AntiAliasing}.Outer)
  )
  NamedRerouteDeclaration(None=Min(A=Min(A=@n14, B=@n15), B=OneMinus(None=@n21))) {name=Shape, nodeColor=<rgba 0 0 0 1>}
 ]
}
