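# HUD element widget that shows the Battle Royale warmup status: a large countdown text above a timer icon and uppercase description, framed by mirrored, sheared brackets.
# Exposed functions set the time text, the description and countdown visibility; the event graph has no logic, so values come from outside.
# TriggerIntroAnim plays Anim_Intro, which staggers fades, slides and scale pops, and drives font, outline and bracket material parameters.
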WidgetBlueprint /BattleRoyaleGameUI/HUD/GamePhase/WBP_BRGameUI_WarmupStatus
animations [Anim_Intro {
  binding_count=5
  length_seconds=1.45
  object=&self:Anim_Intro
  sequence {bindings [
    Text_Time {
     id=#1
     possessed_class=CommonTextBlock
     tracks [
      {
       type=MovieSceneWidgetMaterialTrack
       display_name=Font.FontMaterial
       sections [{
         type=MovieSceneParameterSection
         channels [
          Max_Alpha {type=MovieSceneScriptingFloatChannel, keys [[3,0],[5,1]]}
          Min_Alpha {type=MovieSceneScriptingFloatChannel, keys [[3,0],[9,1],[29,0.5]]}
         ]
        }]
      }
      {
       type=MovieSceneWidgetMaterialTrack
       display_name=Font.Outline Settings.OutlineMaterial
       sections [{type=MovieSceneParameterSection, channels [AnimOpacity {type=MovieSceneScriptingFloatChannel, keys [[3,0],[19,1]]}]}]
      }
      {
       type=MovieScene2DTransformTrack
       sections [{
         type=MovieScene2DTransformSection
         channels [
          Translation.X {type=MovieSceneScriptingFloatChannel, default=0}
          Translation.Y {type=MovieSceneScriptingFloatChannel, keys [[3,48],[19,0]]}
          Scale.X {type=MovieSceneScriptingFloatChannel, keys [[3,0.5],[19,1]]}
          Scale.Y {type=MovieSceneScriptingFloatChannel, keys [[3,0.5],[19,1]]}
         ]
        }]
      }
     ]
    }
    Image_TimerIcon {
     id=#2
     possessed_class=Image
     tracks [
      {
       type=MovieScene2DTransformTrack
       sections [{
         type=MovieScene2DTransformSection
         channels [
          Scale.X {type=MovieSceneScriptingFloatChannel, keys [[3,0.5],[5,2],[18,1]]}
          Scale.Y {type=MovieSceneScriptingFloatChannel, keys [[3,0.5],[5,2],[18,1]]}
         ]
        }]
      }
      {
       type=MovieSceneFloatTrack
       display_name=Render Opacity
       sections [{type=MovieSceneFloatSection, channels [None {type=MovieSceneScriptingFloatChannel, keys [[3,0],[7,1]]}]}]
      }
     ]
    }
    Text_Description {
     id=#3
     possessed_class=CommonTextBlock
     tracks [
      {
       type=MovieSceneFloatTrack
       display_name=Render Opacity
       sections [{type=MovieSceneFloatSection, channels [None {type=MovieSceneScriptingFloatChannel, keys [[0,0],[8,1]]}]}]
      }
      {
       type=MovieScene2DTransformTrack
       sections [{
         type=MovieScene2DTransformSection
         channels [
          Translation.X {type=MovieSceneScriptingFloatChannel, keys [[0,-192],[8,0]]}
          Scale.X {type=MovieSceneScriptingFloatChannel, keys [[0,0.75],[8,1]]}
          Scale.Y {type=MovieSceneScriptingFloatChannel, keys [[0,0.75],[8,1]]}
         ]
        }]
      }
     ]
    }
    Image_BracketRight {
     id=#4
     possessed_class=Image
     tracks [
      {
       type=MovieSceneWidgetMaterialTrack
       sections [{
         type=MovieSceneParameterSection
         channels [
          Anim_Color {type=MovieSceneScriptingFloatChannel, keys [[5,1],[25,0]]}
          Anim_Opacity {type=MovieSceneScriptingFloatChannel, keys [[5,0.5],[13,0]]}
         ]
        }]
      }
      {
       type=MovieSceneFloatTrack
       display_name=Render Opacity
       sections [{type=MovieSceneFloatSection, channels [None {type=MovieSceneScriptingFloatChannel, keys [[5,0],[13,1]]}]}]
      }
     ]
    }
    Image_BracketLeft {
     id=#5
     possessed_class=Image
     tracks [
      {
       type=MovieSceneFloatTrack
       display_name=Render Opacity
       sections [{type=MovieSceneFloatSection, channels [None {type=MovieSceneScriptingFloatChannel, keys [[5,0],[13,1]]}]}]
      }
      {
       type=MovieSceneWidgetMaterialTrack
       sections [{type=MovieSceneParameterSection, channels [Anim_Color {type=MovieSceneScriptingFloatChannel, keys [[5,1],[25,0]]}]}]
      }
     ]
    }
   ]}
 }]
graphs [
 EventGraph {nodes [
   Event PreConstruct()
   Event Construct()
   Event Tick()
  ]}
 SetCountdownNumberVisibility {nodes [
   n1 = SetCountdownNumberVisibility!() -> (then, InVisibility:ESlateVisibility Enum=Visible) {is=title K2Node_FunctionEntry}
   > SetVisibility(self=Get Text_Time() -> (Text_Time:Common Text Object Reference), InVisibility=@n1)
  ]}
 SetDescription {nodes [
   n1 = SetDescription!() -> (then, InText:Text) {is=title K2Node_FunctionEntry}
   > "SetText (Text)"(self=Get Text_Description() -> (Text_Description:Common Text Object Reference), InText=@n1)
  ]}
 SetTimeText {nodes [
   n1 = SetTimeText!() -> (then, InText:Text) {is=title K2Node_FunctionEntry}
   > "SetText (Text)"(self=Get Text_Time(), InText=@n1)
  ]}
 TriggerIntroAnim {nodes [
   TriggerIntroAnim() {is=title K2Node_FunctionEntry}
   > PlayAnimation(InAnimation=Get Anim_Intro() -> (Anim_Intro:Widget Animation Object Reference))
  ]}
]
variables []
widgets {
 info {inheritedWidgetCount=1, parentClass=&FortHUDElementWidget, rootWidgetClass=&Overlay, widgetCount=14}
 widgets [
  Overlay_27 {widgetClassPath=&Overlay, widget=Overlay_27}
  HorizontalBox_0 {
   widgetClassPath=&HorizontalBox
   parent=Overlay_27
   widget=HorizontalBox_0
   slot {class=OverlaySlot, properties {verticalAlignment=VAlign_Center}}
  }
  Spacer {widgetClassPath=&Spacer, parent=HorizontalBox_0, widget=Spacer, properties {size=<60 1>}, slot {class=HorizontalBoxSlot}}
  Bracket_Left {
   widgetClassPath=&FortVisualAttachment
   parent=HorizontalBox_0
   widget=Bracket_Left
   properties {contentAnchor=<0.5 0.6>}
   slot {class=HorizontalBoxSlot, properties {horizontalAlignment=HAlign_Left, verticalAlignment=VAlign_Center}}
  }
  Image_BracketLeft {
   widgetClassPath=&Image
   parent=Bracket_Left
   widget=Image_BracketLeft
   properties {
    pixelSnapping=Disabled
    visibility=SelfHitTestInvisible
    brush {
     bIsDynamicallyLoaded=false
     drawAs=Image
     imageType=FullColor
     mirroring=NoMirror
     resourceName=None
     resourceObject=/BattleRoyaleGameUI/Shared/FrameDecor/MI_UI_BarBracket_GamePhase
     tiling=NoTile
     imageSize=<88 96>
     margin=<ltrb 0 0 0 0>
     outlineSettings {
      bUseBrushTransparency=false
      roundingType=HalfHeightRadius
      width=0
      color {colorUseRule=UseColor_Specified, specifiedColor=<rgba 0 0 0 0>}
      cornerRadii=<0 0 0 0>
     }
     tintColor {colorUseRule=UseColor_Specified, specifiedColor=<rgba 1 1 1 1>}
     uVRegion {bIsValid=false, max=<0 0>, min=<0 0>}
    }
    renderTransform {angle=0, scale=<1 -1>, shear=<8 0>, translation=<0 0>}
   }
   slot {class=SizeBoxSlot}
  }
  VerticalBox_0 {widgetClassPath=&VerticalBox, parent=HorizontalBox_0, widget=VerticalBox_0, slot {class=HorizontalBoxSlot}}
  Text_Time {
   widgetClassPath=&CommonTextBlock
   bIsVariable=true
   parent=VerticalBox_0
   widget=Text_Time
   properties {
    justification=Center
    mobileFontSizeMultiplier=0.85
    text="{####)s"
    font {
     bForceMonospaced=false
     bMaterialIsStencil=false
     fontMaterial=/BattleRoyaleGameUI/Shared/M_UI_Font_FlavorText_Master
     fontObject=&Burbank
     letterSpacing=0
     monospacedWidth=1
     size=48
     skewAmount=0.25
     typefaceFontName=BurbankRegular-Black
     outlineSettings {
      bApplyOutlineToDropShadows=false
      bMiteredCorners=false
      bSeparateFillAlpha=false
      outlineMaterial=/BattleRoyaleGameUI/Shared/MI_UI_Font_SoftShadow_LargeText
      outlineSize=16
      outlineColor=<rgba 1 1 1 1>
     }
    }
    margin=<ltrb 0 -24 0 -24>
    shadowOffset=<0 0>
    strikeBrush {
     bIsDynamicallyLoaded=false
     drawAs=Image
     imageType=NoImage
     mirroring=NoMirror
     resourceName=None
     resourceObject=null
     tiling=NoTile
     imageSize=<32 32>
     margin=<ltrb 0 0 0 0>
     outlineSettings {
      bUseBrushTransparency=false
      roundingType=HalfHeightRadius
      width=0
      color {colorUseRule=UseColor_Specified, specifiedColor=<rgba 0 0 0 0>}
      cornerRadii=<0 0 0 0>
     }
     tintColor {colorUseRule=UseColor_Specified, specifiedColor=<rgba 0.9 0.9 0.9 1>}
     uVRegion {bIsValid=false, max=<0 0>, min=<0 0>}
    }
   }
   slot {class=VerticalBoxSlot, properties {horizontalAlignment=HAlign_Center, verticalAlignment=VAlign_Top}}
  }
  HorizontalBox_40 {
   widgetClassPath=&HorizontalBox
   parent=VerticalBox_0
   widget=HorizontalBox_40
   slot {class=VerticalBoxSlot, properties {padding=<ltrb 0 -4 0 0>}}
  }
  Image_TimerIcon {
   widgetClassPath=&Image
   parent=HorizontalBox_40
   widget=Image_TimerIcon
   properties {
    visibility=SelfHitTestInvisible
    brush {
     bIsDynamicallyLoaded=false
     drawAs=Image
     imageType=FullColor
     mirroring=NoMirror
     resourceName=None
     resourceObject=&/M_UI_GamePhaseTimer
     tiling=NoTile
     imageSize=<40 40>
     margin=<ltrb 0 0 0 0>
     outlineSettings {
      bUseBrushTransparency=false
      roundingType=HalfHeightRadius
      width=0
      color {colorUseRule=UseColor_Specified, specifiedColor=<rgba 0 0 0 0>}
      cornerRadii=<0 0 0 0>
     }
     tintColor {colorUseRule=UseColor_Specified, specifiedColor=<rgba 1 1 1 1>}
     uVRegion {bIsValid=false, max=<0 0>, min=<0 0>}
    }
   }
   slot {class=HorizontalBoxSlot, properties {horizontalAlignment=HAlign_Center, verticalAlignment=VAlign_Center}}
  }
  Text_Description {
   widgetClassPath=&CommonTextBlock
   bIsVariable=true
   parent=HorizontalBox_40
   widget=Text_Description
   properties {
    minDesiredWidth=1
    mobileFontSizeMultiplier=1.15
    text=Text to Display
    textTransformPolicy=ToUpper
    font {
     bForceMonospaced=false
     bMaterialIsStencil=false
     fontMaterial=/BattleRoyaleGameUI/Shared/M_UI_Font_White
     fontObject=&Burbank
     letterSpacing=20
     monospacedWidth=1
     size=21
     skewAmount=0
     typefaceFontName=BurbankRegular-Black
     outlineSettings {
      bApplyOutlineToDropShadows=false
      bMiteredCorners=false
      bSeparateFillAlpha=true
      outlineMaterial=&MI_UI_Font_SoftShadow_LargeText
      outlineSize=8
      outlineColor=<rgba 1 1 1 1>
     }
    }
    margin=<ltrb -6 -6 -6 -8>
    shadowOffset=<0 0>
   }
   slot {class=HorizontalBoxSlot, properties {padding=<ltrb 0 2 0 0>}}
  }
  Bracket_Right {
   widgetClassPath=&FortVisualAttachment
   parent=HorizontalBox_0
   widget=Bracket_Right
   properties {contentAnchor=<0.5 0.6>}
   slot {class=HorizontalBoxSlot, properties {horizontalAlignment=HAlign_Right, verticalAlignment=VAlign_Center}}
  }
  Image_BracketRight {
   widgetClassPath=&Image
   parent=Bracket_Right
   widget=Image_BracketRight
   properties {
    pixelSnapping=Disabled
    visibility=SelfHitTestInvisible
    brush {
     bIsDynamicallyLoaded=false
     drawAs=Image
     imageType=FullColor
     mirroring=NoMirror
     resourceName=None
     resourceObject=&MI_UI_BarBracket_GamePhase
     tiling=NoTile
     imageSize=<88 96>
     margin=<ltrb 0 0 0 0>
     outlineSettings {
      bUseBrushTransparency=false
      roundingType=HalfHeightRadius
      width=0
      color {colorUseRule=UseColor_Specified, specifiedColor=<rgba 0 0 0 0>}
      cornerRadii=<0 0 0 0>
     }
     tintColor {colorUseRule=UseColor_Specified, specifiedColor=<rgba 1 1 1 1>}
     uVRegion {bIsValid=false, max=<0 0>, min=<0 0>}
    }
    renderTransform {angle=0, scale=<-1 -1>, shear=<-8 0>, translation=<0 0>}
   }
   slot {class=SizeBoxSlot}
  }
  Spacer_187 {widgetClassPath=&Spacer, parent=HorizontalBox_0, widget=Spacer_187, properties {size=<60 1>}, slot {class=HorizontalBoxSlot}}
  TutorialHighlight {widgetClassPath=&TutorialHighlight}
 ]
}
